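# Control Rig for the Love-A-Flare gadget prop, used to animate its root, offset, main, two heart and flare bones through controls.
# The forward solve loops over a fixed list of these six bones and copies each matching control's transform onto the bone.
# Controls are matched by stripping the "JNT_" prefix from bone names.
# Each heart and flare control sits under its own "_Main" parent control for layered adjustment.
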ControlRigBlueprint /BRCosmetics/Animation/Props/Gadgets_LoveAFlare/Meshes/CR_LoveAFlare_Gadget
graphs [
 RigVMFunctionLibrary {links [], nodes []}
 RigVMModel {nodes [
   RigUnit_BeginExecution = &RigUnit_BeginExecution()
   RigUnit_ItemArray = &RigUnit_ItemArray(Items=((Type=Bone,Name="root"),(Type=Bone,Name="offset"),(Type=Bone,Name="main"),(Type=Bone,Name="JNT_Heart1"),(Type=Bone,Name="JNT_Heart2"),(Type=Bone,Name="JNT_Flare")))
   For_Each = RigVMDispatchNode(ExecuteContext=@RigUnit_BeginExecution.ExecutePin, Array=@RigUnit_ItemArray.Items) {template="DISPATCH_RigVMDispatch_ArrayIterator(in Array,out Element,out Index,out Count,out Ratio)"}
   Replace = /Script/RigVM.RigVMFunction_NameReplace(Old=JNT_, Name=@For_Each.Element.Name) {is=struct RigVMUnitNode, template="Replace::Execute(in Name,in Old,in New,out Result)"}
   > &RigUnit_SetTransform(
    Weight=1
    bInitial=False
    Item=@For_Each.Element
    Value=&RigUnit_GetTransform(Item=(Type=Control,Name="None"), bInitial=False, Item.Name=@Replace.Result).Transform
   )
  ]}
]
hierarchy [
 root {loc=<0 0 0>, rot=<0 0 0 1>}
 offset {parent=root, loc=<0 0 0>, rot=<0 0 0 1>}
 main {parent=offset, loc=<0 0 0>, rot=<0 0 0 1>}
 JNT_Heart1 {parent=main, loc=<0 0 0>, rot=<0 0 0 1>}
 JNT_Heart2 {parent=main, loc=<0 0 0>, rot=<0 0 0 1>}
 JNT_Flare {parent=main, loc=<0 0 0>, rot=<0 0 0 1>}
 root {controlType=EULER_TRANSFORM, type=Control, loc=<0 0 0>, rot=<0 0 0 1>}
 offset {controlType=EULER_TRANSFORM, parent=root, type=Control, loc=<0 0 0>, rot=<0 0 0 1>}
 main {controlType=EULER_TRANSFORM, parent=offset, type=Control, loc=<0 0 0>, rot=<0 0 0 1>}
 Flare_Main {controlType=EULER_TRANSFORM, parent=main, type=Control, loc=<0 0 0>, rot=<0 0 0 1>}
 Flare {controlType=EULER_TRANSFORM, parent=Flare_Main, type=Control, loc=<0 0 0>, rot=<0 0 0 1>}
 Heart1_Main {controlType=EULER_TRANSFORM, parent=main, type=Control, loc=<0 0 0>, rot=<0 0 0 1>}
 Heart1 {controlType=EULER_TRANSFORM, parent=Heart1_Main, type=Control, loc=<0 0 0>, rot=<0 0 0 1>}
 Heart2_Main {controlType=EULER_TRANSFORM, parent=main, type=Control, loc=<0 0 0>, rot=<0 0 0 1>}
 Heart2 {controlType=EULER_TRANSFORM, parent=Heart2_Main, type=Control, loc=<0 0 0>, rot=<0 0 0 1>}
]
